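# Conversation database for the AuraCop NPC, providing its greeting and top-level dialogue menu.
# A root Speech node picks contextual lines by first-ever meeting, match conversation count or player outfit, using curve-row priorities.
# It branches through tag-based link nodes to shared quests and basic services, plus AuraCop dialogue and Season 36 character stories.
# A DMR Expertise choice opens a submenu linking to the Expertise C entry, with a Back option.
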ConversationDatabase /NPCLibrary/NPCs/AuraCop/NPCLibrary_Conversation_AuraCop
entry_points [{tag=Conversation.Entry.AuraCop, destinations [#1]}]
nodes [
 ConversationEntryPointNode_7 {
  id=#1
  class=ConversationEntryPointNode
  outputs [#2]
  properties {entryTag {tagName=Conversation.Entry.AuraCop}}
 }
 Speech_C_2 {
  id=#2
  class=Speech_C
  outputs [
   #3
   #4
   #5
   #6
   #7
  ]
  properties {
   nodeName=""
   generalConfig {
    defaultMessage=I've always done my best work from the SHADOWs.
    contextualMessages [
     {
      message=You already know who I am.
      contextRequirements [{requirement {class=FortControllerRequirement_IsFirstEverConversationWithNPC}}]
      priority {curve {rowName=Tandem.Speech.Priority.FirstEverInteraction}}
     }
     {
      message=I've always done my best work from the SHADOWs.
      contextRequirements [{requirement {class=FortControllerRequirement_CheckNPCMatchConversationCount, properties {nPCUniqueIdTag {tagName=AISpawnerData.Type.Tandem.AuraCop}}}}]
      priority {curve {rowName=None}}
     }
     {
      message=Welcome to Utopia City. I may have aided in funding its construction.
      contextRequirements [{requirement {class=FortControllerRequirement_CheckNPCMatchConversationCount, properties {nPCUniqueIdTag {tagName=AISpawnerData.Type.Tandem.AuraCop}}}}]
      priority {curve {rowName=None}}
     }
     {
      message=Grand plans require unusual allies.
      contextRequirements [{requirement {class=FortControllerRequirement_CheckNPCMatchConversationCount, properties {nPCUniqueIdTag {tagName=AISpawnerData.Type.Tandem.AuraCop}}}}]
      priority {curve {rowName=None}}
     }
     {
      message="Kane is dealt with. Next, his mask maker."
      contextRequirements [{requirement {class=FortControllerRequirement_CheckNPCMatchConversationCount, properties {nPCUniqueIdTag {tagName=AISpawnerData.Type.Tandem.AuraCop}}}}]
      priority {curve {rowName=None}}
     }
     {
      message="We have a common goal, despite our differing motives."
      contextRequirements [{requirement {class=FortControllerRequirement_HasCID, properties {allowedCIds [&Character_IronLilac_Sly]}}}]
      priority {curve {rowName=Tandem.Speech.Priority.OutfitReaction}}
     }
     {
      message="No hard feelings, hm? Let's leave the past in the past."
      contextRequirements [{requirement {
         class=FortControllerRequirement_HasCID
         properties {allowedCIds [&Character_UndergroundRebel_Fashion]}
        }}]
      priority {curve {rowName=Tandem.Speech.Priority.OutfitReaction}}
     }
     {
      message=Your recklessness has cost you greatly. My... sympathies.
      contextRequirements [{requirement {
         class=FortControllerRequirement_HasCID
         properties {allowedCIds [&Character_MistMylar_Claw, /BRCosmetics/Athena/Items/Cosmetics/Characters/Character_SolidMist]}
        }}]
      priority {curve {rowName=Tandem.Speech.Priority.OutfitReaction}}
     }
     {
      message="Remember this, Kane: You lost the moment you came for me."
      contextRequirements [{requirement {class=FortControllerRequirement_HasCID, properties {allowedCIds [/BRCosmetics/Athena/Items/Cosmetics/Characters/Character_SherWolf_Gnash]}}}]
      priority {curve {rowName=Tandem.Speech.Priority.OutfitReaction}}
     }
     {
      message="I know it's not your forte, but let's keep our continued dealings quiet."
      contextRequirements [{requirement {class=FortControllerRequirement_HasCID, properties {allowedCIds [/BRCosmetics/Athena/Items/Cosmetics/Characters/Character_RageDebris]}}}]
      priority {curve {rowName=Tandem.Speech.Priority.OutfitReaction}}
     }
     {
      message=You want to be my protégé? I'll give you a chance to earn it.
      contextRequirements [{requirement {class=FortControllerRequirement_HasCID, properties {allowedCIds [/BRCosmetics/Athena/Items/Cosmetics/Characters/Character_CasinoReaper_Die]}}}]
      priority {curve {rowName=Tandem.Speech.Priority.OutfitReaction}}
     }
     {
      message="Happy doing business with you, Dillinger."
      contextRequirements [{requirement {class=FortControllerRequirement_HasCID, properties {allowedCIds [/BRCosmetics/Athena/Items/Cosmetics/Characters/Character_WartyBrine_Sus]}}}]
      priority {curve {rowName=Tandem.Speech.Priority.OutfitReaction}}
     }
     {
      message="And why, little bird, are you watching me so carefully?"
      contextRequirements [{requirement {class=FortControllerRequirement_HasCID, properties {allowedCIds [/BRCosmetics/Athena/Items/Cosmetics/Characters/Character_MistRavenPeck]}}}]
      priority {curve {rowName=Tandem.Speech.Priority.OutfitReaction}}
     }
     {
      message="I hope you're finding a warmer life, outside of my shadow."
      contextRequirements [{requirement {
         class=FortControllerRequirement_HasCID
         properties {allowedCIds [/BRCosmetics/Athena/Items/Cosmetics/Characters/CID_770_Athena_Commando_F_MechanicalEngineer]}
        }}]
      priority {curve {rowName=Tandem.Speech.Priority.OutfitReaction}}
     }
     {
      message="No matter what happens, you and I will always stand side by side."
      contextRequirements [{requirement {
         class=FortControllerRequirement_HasCID
         properties {allowedCIds [/BRCosmetics/Athena/Items/Cosmetics/Characters/CID_A_034_Athena_Commando_F_CatBurglar]}
        }}]
      priority {curve {rowName=Tandem.Speech.Priority.OutfitReaction}}
     }
     {
      message="Keep me apprised of the crew's movements, Brutus."
      contextRequirements [{requirement {
         class=FortControllerRequirement_HasCID
         properties {allowedCIds [&CID_692_Athena_Commando_M_HenchmanTough]}
        }}]
      priority {curve {rowName=Tandem.Speech.Priority.OutfitReaction}}
     }
     {
      message="You're the best at what you do, and I accept nothing less."
      contextRequirements [{requirement {class=FortControllerRequirement_HasCID, properties {allowedCIds [/BRCosmetics/Athena/Items/Cosmetics/Characters/Character_RoyalAngst_Tote]}}}]
      priority {curve {rowName=Tandem.Speech.Priority.OutfitReaction}}
     }
     {
      message=Good luck finding a scientific explanation for the golden curse.
      contextRequirements [{requirement {class=FortControllerRequirement_HasCID, properties {allowedCIds [/BRCosmetics/Athena/Items/Cosmetics/Characters/Character_TrafficHat_Boba]}}}]
      priority {curve {rowName=Tandem.Speech.Priority.OutfitReaction}}
     }
     {
      message="Your friend Cassidy shows real potential, if she's willing to ante up."
      contextRequirements [{requirement {class=FortControllerRequirement_HasCID, properties {allowedCIds [/BRCosmetics/Athena/Items/Cosmetics/Characters/Character_TealMink_Pelt]}}}]
      priority {curve {rowName=Tandem.Speech.Priority.OutfitReaction}}
     }
    ]
   }
  }
 }
 ConversationLinkNode_3 {
  id=#7
  class=ConversationLinkNode
  properties {remoteEntryTag {tagName=Conversation.Entry.Season36.AuraCop.CharacterStories}}
 }
 Speech_C_3 {
  id=#6
  class=Speech_C
  outputs [#8, #9]
  properties {softTaskIcon=&T_UI_NPC_WeaponExpertise, generalConfig {defaultMessage=What do you want?}}
  sub_nodes [ConversationChoiceNode_4 {class=ConversationChoiceNode, properties {defaultChoiceDisplayText=DMR Expertise}}]
 }
 ConversationLinkNode_16 {id=#9, class=ConversationLinkNode, properties {remoteEntryTag {tagName=Conversation.Entry.Expertise.C}}}
 Back_C_0 {
  id=#8
  class=Back_C
  sub_nodes [ConversationChoiceNode_1 {class=ConversationChoiceNode, properties {defaultChoiceDisplayText=BACK}}]
 }
 ConversationLinkNode_2 {id=#5, class=ConversationLinkNode, properties {remoteEntryTag {tagName=Conversation.Entry.AuraCop.Dialogue}}}
 ConversationLinkNode_0 {id=#4, class=ConversationLinkNode, properties {remoteEntryTag {tagName=Conversation.Entry.BasicServices}}}
 ConversationLinkNode_1 {id=#3, class=ConversationLinkNode, properties {remoteEntryTag {tagName=Conversation.Entry.NPCLibrary.Quests}}}
]
